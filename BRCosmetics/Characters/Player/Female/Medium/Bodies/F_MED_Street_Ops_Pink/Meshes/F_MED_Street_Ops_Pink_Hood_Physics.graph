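# Physics asset for the hood mesh of the F_MED_Street_Ops_Pink female player cosmetic, covering a hood and two arm zipper bones.
# Six capsule bodies cover those dyn_ bones, both upper-arm twist bones and spine_05, whose large zero-length capsule acts as a torso sphere.
# The hood constrains to spine_05 and each zipper to its upper arm with tight twist limits.
# Both upper arms link to spine_05 with free swing and twist.
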PhysicsAsset /BRCosmetics/Characters/Player/Female/Medium/Bodies/F_MED_Street_Ops_Pink/Meshes/F_MED_Street_Ops_Pink_Hood_Physics
bodies {
 dyn_arm_zipper_L [{shapeType=Capsule, length=2, radius=1, center=<0 0.0001 -0.2564>}]
 dyn_arm_zipper_R [{shapeType=Capsule, length=1.75, radius=1.25, center=<0 0.0001 -0.2564>}]
 dyn_hood [{shapeType=Capsule, length=5, radius=3.25, center=<1.8972 0 -0.2>, rotation=<rot -59.9953 179.9999 180>}]
 spine_05 [{shapeType=Capsule, length=0, radius=29, center=<-8.4399 17.1685 0.0002>}]
 upperarm_twist_01_l [{shapeType=Capsule, length=1, radius=1}]
 upperarm_twist_01_r [{shapeType=Capsule, length=1, radius=1}]
}
constraints [
 {bone1Name=dyn_hood, bone2Name=spine_05, swing1LimitDegrees=84.4286, swing2LimitDegrees=63.8571, twistLimitDegrees=5}
 {
  bone1Name=upperarm_twist_01_l
  bone2Name=spine_05
  swing1LimitDegrees=45
  swing1Motion=Free
  swing2LimitDegrees=45
  swing2Motion=Free
  twistLimitDegrees=45
  twistMotion=Free
 }
 {
  bone1Name=upperarm_twist_01_r
  bone2Name=spine_05
  swing1LimitDegrees=45
  swing1Motion=Free
  swing2LimitDegrees=45
  swing2Motion=Free
  twistLimitDegrees=45
  twistMotion=Free
 }
 {bone1Name=dyn_arm_zipper_L, bone2Name=upperarm_twist_01_l, swing1LimitDegrees=45, swing2LimitDegrees=45, twistLimitDegrees=5}
 {bone1Name=dyn_arm_zipper_R, bone2Name=upperarm_twist_01_r, swing1LimitDegrees=45, swing2LimitDegrees=45, twistLimitDegrees=5}
]
