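# Physics asset for the pack-equipped mesh of the female medium Street Ops Pink body cosmetic, covering its dynamic jacket bones.
# Box bodies on the front, left and right jacket bones are each constrained to spine_03, which carries two capsules as the torso collision shape.
# Angular limits keep twist tight, with more swing on the first axis than the second and slightly more for the side pieces than the front.
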PhysicsAsset /BRCosmetics/Characters/Player/Female/Medium/Bodies/F_MED_Street_Ops_Pink/Meshes/F_MED_Street_Ops_Pink_PackEquipped_Physics
bodies {
 dyn_jacket_front [{shapeType=Box, extentX=8.486, extentY=2.0998, extentZ=13.6942, center=<2.9141 0.3013 0>}]
 dyn_jacket_l [{shapeType=Box, extentX=8.486, extentY=12.5998, extentZ=2.1942, center=<3.2405 0 0.0001>}]
 dyn_jacket_r [{shapeType=Box, extentX=8.486, extentY=12.5998, extentZ=2.1942, center=<3.2405 0 0.0001>}]
 spine_03 [
  {shapeType=Capsule, length=11.75, radius=7.5, center=<-0.11 2.5688 0.0001>, rotation=<rot 90 -0.5968 -5.5968>}
  {shapeType=Capsule, length=7, radius=5.5, center=<-2.9695 0.597 -0.0254>, rotation=<rot 0.2247 7.1286 -0.1929>}
 ]
}
constraints [
 {bone1Name=dyn_jacket_front, bone2Name=spine_03, swing1LimitDegrees=30, swing2LimitDegrees=10, twistLimitDegrees=5}
 {bone1Name=dyn_jacket_l, bone2Name=spine_03, swing1LimitDegrees=35, swing2LimitDegrees=10, twistLimitDegrees=5}
 {bone1Name=dyn_jacket_r, bone2Name=spine_03, swing1LimitDegrees=35, swing2LimitDegrees=10, twistLimitDegrees=5}
]
