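# Post-process style AnimBP for the female Street Ops Pink body, copying the parent mesh pose and adding secondary motion.
# Thigh rotation drives glute corrective bones, damped while moving; AnimDynamics swings shoe tongues and foot zippers before caching.
# Two RigidBody body/hood chains branch from the cache, selected by isBackStatic (backpack, disabled rigid bodies or high LOD).
# A weighted constraint then positions the jacket zipper before AnimDynamics simulates it.
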AnimBlueprint /BRCosmetics/Characters/Player/Female/Medium/Bodies/F_MED_Street_Ops_Pink/Meshes/F_MED_Street_Ops_Pink_AnimBP
target_skeleton=&/F_MED_Street_Ops_Pink_Skeleton
graphs [
 AnimGraph {nodes [
   n1 = Anim Dynamics - Bone: dyn_arm_zipper_R(
    ComponentPose=Reroute Node(InputPin:Component Space Pose Link Structure) -> (OutputPin:Component Space Pose Link Structure)
    Alpha=0.15
    ExternalForce="0,0,0"
   ) {
    bGravityOverrideInSimSpace=true
    lODThreshold=1
    numSolverIterationsPostUpdate=4
    numSolverIterationsPreUpdate=16
    simulationSpace=Actor
    alphaBoolBlend {blendOption=Linear}
    boundBone {boneName=dyn_arm_zipper_R}
    physicsBodyDefinitions [{
      sphereCollisionRadius=0
      boundBone {boneName=dyn_arm_zipper_R}
      boxExtents=<5 5 5>
      constraintSetup {angularTargetAxis=AxisZ, twistAxis=AxisZ, angularLimitsMax=<5 30 30>, angularLimitsMin=<-90 0 -30>, angularTarget=<0 0 0>}
      localJointOffset=<0 0 -0.5>
     }]
   }
   Anim Dynamics - Bone: dyn_arm_zipper_L(ComponentPose=@n1, Alpha=0.25, ExternalForce="0,0,0") {
    angularDampingOverride=0.9999
    bGravityOverrideInSimSpace=true
    bOverrideAngularDamping=true
    gravityScale=0.35
    lODThreshold=1
    numSolverIterationsPostUpdate=4
    numSolverIterationsPreUpdate=16
    simulationSpace=RootRelative
    alphaBoolBlend {blendOption=Linear}
    boundBone {boneName=dyn_arm_zipper_L}
    componentLinearAccScale=<0.4 0.4 0.4>
    componentLinearVelScale=<0.4 0.4 0.4>
    physicsBodyDefinitions [{
      sphereCollisionRadius=0
      boundBone {boneName=dyn_arm_zipper_L}
      boxExtents=<5 10 5>
      constraintSetup {angularTargetAxis=AxisZ, twistAxis=AxisZ, angularLimitsMax=<30 90 2>, angularLimitsMin=<-30 -90 -2>, angularTarget=<0 0 0>}
      localJointOffset=<1 0 -3>
     }]
   }
   n2 = RigidBody(ComponentPose=Local To Component(LocalPose=Use cached pose 'AnimCache'())) {
    bClampLinearTranslationLimitToRefPose=true
    bOverrideWorldGravity=true
    overridePhysicsAsset=&/F_MED_Street_Ops_Pink_PackEquipped_Physics
    simulationSpace=BaseBoneSpace
    alphaBoolBlend {blendOption=Linear}
    baseBoneRef {boneName=pelvis}
    componentAppliedLinearAccClamp=<6000 6000 6000>
    componentLinearAccScale=<0.5 0.5 0.5>
    componentLinearVelScale=<0.5 0.5 0.5>
    overrideWorldGravity {z=-980}
   }
   n3 = RigidBody(ComponentPose=@n2, Alpha=0.2) {
    bClampLinearTranslationLimitToRefPose=true
    bOverrideWorldGravity=true
    overridePhysicsAsset=&/F_MED_Street_Ops_Pink_Hood_Physics
    simulationSpace=BaseBoneSpace
    alphaBoolBlend {blendOption=Linear}
    baseBoneRef {boneName=root}
    componentAppliedLinearAccClamp {z=2000}
    componentLinearAccScale=<0.5 0.5 0.45>
    componentLinearVelScale=<0.5 0.5 0.45>
    overrideWorldGravity {z=-700}
   }
   n4 = Component To Local(ComponentPose=@n3)
   n5 = RigidBody(ComponentPose=Local To Component(LocalPose=Use cached pose 'AnimCache'())) {
    bClampLinearTranslationLimitToRefPose=true
    bOverrideWorldGravity=true
    overridePhysicsAsset=&/F_MED_Street_Ops_Pink_Physics
    simulationSpace=BaseBoneSpace
    alphaBoolBlend {blendOption=Linear}
    baseBoneRef {boneName=pelvis}
    componentAppliedLinearAccClamp=<6000 6000 6000>
    componentLinearAccScale=<0.5 0.5 0.5>
    componentLinearVelScale=<0.5 0.5 0.5>
    overrideWorldGravity {z=-980}
   }
   n6 = RigidBody(ComponentPose=@n5) {
    bClampLinearTranslationLimitToRefPose=true
    bOverrideWorldGravity=true
    overridePhysicsAsset=&F_MED_Street_Ops_Pink_Hood_Physics
    simulationSpace=BaseBoneSpace
    alphaBoolBlend {blendOption=Linear}
    baseBoneRef {boneName=root}
    componentAppliedLinearAccClamp {z=2000}
    componentLinearAccScale=<0.5 0.5 0.45>
    componentLinearVelScale=<0.5 0.5 0.45>
    overrideWorldGravity {z=-700}
   }
   n7 = Component To Local(ComponentPose=@n6)
   n8 = Get isBackStatic() -> (isBackStatic:bool)
   n9 = "Constraint - parent_dyn_zipper (PPP)"!(
    ComponentPose=Local To Component(LocalPose=Blend Poses by bool(BlendPose_0=@n4, BlendPose_1=@n7, bActiveValue=@n8) {blendType=Linear})
    Alpha=1
    ConstraintWeights_0=0.1709
    ConstraintWeights_1:float=0.2561
    ConstraintWeights_2:float=0.573
   ) -> (Pose) {
    alphaBoolBlend {blendOption=Linear}
    boneToModify {boneName=parent_dyn_zipper}
    constraintSetup [{targetBone {boneName=dyn_jacket_l}}, {targetBone {boneName=spine_04}}, {targetBone {boneName=dyn_jacket_front}}]
    constraintWeights [1,1,1]
   }
   n10 = Anim Dynamics - Bone: dyn_zipper(ComponentPose=@n9, Alpha=1) {
    bGravityOverrideInSimSpace=true
    lODThreshold=1
    numSolverIterationsPostUpdate=4
    numSolverIterationsPreUpdate=16
    simulationSpace=Actor
    alphaBoolBlend {blendOption=Linear}
    boundBone {boneName=dyn_zipper}
    physicsBodyDefinitions [{
      sphereCollisionRadius=0
      boundBone {boneName=dyn_zipper}
      boxExtents=<5 5 5>
      constraintSetup {angularTargetAxis=AxisZ, twistAxis=AxisZ, angularLimitsMax=<5 30 0>, angularLimitsMin=<-90 -30 0>, angularTarget=<0 0 0>}
      localJointOffset=<0 0 -0.5>
     }]
   }
   Output Pose(Result=Component To Local(ComponentPose=@n10)) {layerGroup=None}
   n11 = "Transform (Modify) Bone - Bone: dyn_glute_l"(
    ComponentPose=Local To Component(LocalPose=Copy Pose From Mesh() {bUseAttachedParent=true})
    Alpha=Get thigh_rot_l()
    Translation="-3.750,0.875,-0.9"
    Rotation="0.000,0.000,-55"
    Scale="1.25,1.25,1.25"
   ) {
    rotationMode=BMM_Additive
    rotationSpace=BCS_BoneSpace
    scaleMode=BMM_Additive
    scaleSpace=BCS_BoneSpace
    translationMode=BMM_Additive
    translationSpace=BCS_BoneSpace
    alphaBoolBlend {blendOption=Linear}
    boneToModify {boneName=dyn_glute_l}
    rotation {roll=-55}
    scale=<1.25 1.25 1.25>
    translation=<-3.75 0.875 -0.9>
   }
   n12 = "Transform (Modify) Bone - Bone: dyn_glute_r"(ComponentPose=@n11, Alpha=Get thigh_rot_r(), Translation="-3.75,0,0", Rotation="0.000,0.000,-55", Scale="1.1,1.1,1.1") {
    rotationMode=BMM_Additive
    rotationSpace=BCS_BoneSpace
    scaleMode=BMM_Additive
    scaleSpace=BCS_BoneSpace
    translationMode=BMM_Additive
    translationSpace=BCS_BoneSpace
    alphaBoolBlend {blendOption=Linear}
    boneToModify {boneName=dyn_glute_r}
    rotation {roll=-55}
    scale=<1.1 1.1 1.1>
    translation {x=-3.75}
   }
   n13 = Anim Dynamics - Bone: dyn_tongue_l(ComponentPose=@n12, Alpha=1, ExternalForce="0,0,0") {
    angularBiasOverride=1
    bGravityOverrideInSimSpace=true
    bOverrideAngularBias=true
    gravityScale=0.1
    simulationSpace=BoneRelative
    alphaBoolBlend {blendOption=Linear}
    boundBone {boneName=dyn_tongue_l}
    physicsBodyDefinitions [{
      sphereCollisionRadius=0
      boundBone {boneName=dyn_tongue_l}
      boxExtents=<5 5 5>
      constraintSetup {
       angularTargetAxis=AxisY
       twistAxis=AxisY
       angularLimitsMax=<0 30 0>
       angularLimitsMin=<0 -30 0>
       angularTarget=<0 0 0>
       linearAxesMax=<0 0 3>
      }
      localJointOffset=<0 0 -0.1>
     }]
   }
   n14 = Anim Dynamics - Bone: dyn_tongue_r(ComponentPose=@n13, Alpha=1, ExternalForce="0,0,0") {
    angularBiasOverride=1
    bGravityOverrideInSimSpace=true
    bOverrideAngularBias=true
    gravityScale=0.1
    simulationSpace=BoneRelative
    alphaBoolBlend {blendOption=Linear}
    boundBone {boneName=dyn_tongue_r}
    physicsBodyDefinitions [{
      sphereCollisionRadius=0
      boundBone {boneName=dyn_tongue_r}
      boxExtents=<5 5 5>
      constraintSetup {
       angularTargetAxis=AxisY
       twistAxis=AxisY
       angularLimitsMax=<0 30 0>
       angularLimitsMin=<0 -30 0>
       angularTarget=<0 0 0>
       linearAxesMax=<0 0 3>
      }
      localJointOffset=<0 0 -0.1>
     }]
   }
   n15 = Anim Dynamics - Bone: dyn_foot_zipper_R(ComponentPose=@n14, Alpha=1, ExternalForce="0,0,0") {
    angularDampingOverride=0.999
    bGravityOverrideInSimSpace=true
    bOverrideAngularDamping=true
    lODThreshold=1
    numSolverIterationsPostUpdate=4
    numSolverIterationsPreUpdate=16
    simulationSpace=RootRelative
    alphaBoolBlend {blendOption=Linear}
    boundBone {boneName=dyn_foot_zipper_R}
    physicsBodyDefinitions [{
      sphereCollisionRadius=0
      boundBone {boneName=dyn_foot_zipper_R}
      boxExtents=<5 5 5>
      constraintSetup {angularTargetAxis=AxisZ, twistAxis=AxisZ, angularLimitsMax=<5 0 0>, angularLimitsMin=<-90 -30 0>, angularTarget=<0 0 0>}
      localJointOffset=<0 0 -0.5>
     }]
   }
   n16 = Anim Dynamics - Bone: dyn_foot_zipper_L(ComponentPose=@n15, Alpha=1, ExternalForce="0,0,0") {
    angularDampingOverride=0.999
    bGravityOverrideInSimSpace=true
    bOverrideAngularDamping=true
    lODThreshold=1
    numSolverIterationsPostUpdate=4
    numSolverIterationsPreUpdate=16
    simulationSpace=RootRelative
    alphaBoolBlend {blendOption=Linear}
    boundBone {boneName=dyn_foot_zipper_L}
    physicsBodyDefinitions [{
      sphereCollisionRadius=0
      boundBone {boneName=dyn_foot_zipper_L}
      boxExtents=<5 5 5>
      constraintSetup {angularTargetAxis=AxisZ, twistAxis=AxisZ, angularLimitsMax=<5 0 0>, angularLimitsMin=<-90 -30 0>, angularTarget=<0 0 0>}
      localJointOffset=<0 0 -0.5>
     }]
   }
   Save cached pose 'AnimCache'(Pose=Component To Local(ComponentPose=@n16)) {cacheName=AnimCache}
  ]}
 EventGraph {nodes [
   Event BlueprintUpdateAnimation()
   > Get!(execute:Exec) -> (then:Exec, else:Exec, player_pawn:Fort Player Pawn Athena Object Reference)
   > Get!(execute) -> (then, else, Skeletal_Mesh:Skeletal Mesh Component Object Reference)
   > n1 = Sequence!(execute) -> (then_0, then_1, then_2:Exec)
   > Do Once()
   n2 = Get player_pawn() -> (player_pawn:Fort Player Pawn Athena Object Reference)
   > Set isBackpackEquipped(isBackpackEquipped:bool="Not Equal (Object)"!(A:Object Reference=GetSkeletalMeshForPartType(self=@n2, PartType=Backpack), B:Object Reference) -> (ReturnValue:bool)) -> (Output_Get:bool)
   n3 = Branch(execute=@n1.then_2, Condition="float > float"(A=Get PawnSpeed2D(), B=100) {is=title K2Node_CallFunction})
   n4 = Get Skeletal_Mesh() -> (Skeletal_Mesh:Skeletal Mesh Component Object Reference)
   n5 = GetDeltaTransformFromRefPose(self=@n4, BoneName=thigh_l, BaseName=pelvis)
   n6 = MapRangeClamped(Value=BreakRotator(InRot=BreakTransform(InTransform=@n5).Rotation).Yaw, InRangeA=-20, InRangeB=-100)
   > Set thigh_rot_l(thigh_rot_l=float * float(A=@n6, B=0.25))
   n7 = GetDeltaTransformFromRefPose(self=@n4, BoneName=thigh_r, BaseName=pelvis)
   n8 = MapRangeClamped(Value=BreakRotator(InRot=BreakTransform(InTransform=@n7).Rotation).Yaw, InRangeA=-20, InRangeB=-100)
   > Set thigh_rot_r(thigh_rot_r=float * float(A=@n8, B=0.5))
   Set thigh_rot_l(execute=@n3.else, thigh_rot_l=@n6)
   > Set thigh_rot_r(thigh_rot_r=@n8)
   n9 = OR Boolean(A=NOT Boolean(A=Get bRigidBodyEnabled()), B="integer > integer"(A=Get CurrentLOD(), B=1) {is=title K2Node_CallFunction})
   Set isRigidBodyDisabled(execute=@n1.then_1, isRigidBodyDisabled:bool=@n9) -> (Output_Get:bool)
   > Set isBackStatic(isBackStatic:bool=OR Boolean(A=Get isBackpackEquipped() -> (isBackpackEquipped:bool), B=@n9)) -> (Output_Get:bool)
   Event BlueprintInitializeAnimation()
   n10 = TryGetPawnOwner()
   > Is Valid(InputObject=@n10)
   > n11 = Cast To FortPlayerPawnAthena(Object=@n10)
   > Set player_pawn(player_pawn:Fort Player Pawn Athena Object Reference=@n11) -> (Output_Get:Fort Player Pawn Athena Object Reference)
   > Set Skeletal_Mesh(Skeletal_Mesh:Skeletal Mesh Component Object Reference=GetOwnerSkeletalMeshComponent()) -> (Output_Get:Skeletal Mesh Component Object Reference)
  ]}
]
variables [
 isBackpackEquipped {type=boolean}
 isBackStatic {type=boolean}
 isRigidBodyDisabled {type=boolean}
 player_pawn {type=FortPlayerPawnAthena}
 Skeletal_Mesh {type=SkeletalMeshComponent}
 thigh_rot_l {type=number}
 thigh_rot_r {type=number}
]
